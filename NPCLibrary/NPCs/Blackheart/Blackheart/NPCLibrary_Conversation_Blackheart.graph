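# Conversation database for the Blackheart NPC, a pirate character, defining the dialogue that opens when a player talks to him.
# The entry point leads to one speech node whose default line can be replaced by prioritized contextual messages for a first-ever conversation or for specific player outfits.
# That speech then branches through link nodes to shared NPC quest, basic services and Season 22 narrative quest conversations.
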ConversationDatabase /NPCLibrary/NPCs/Blackheart/Blackheart/NPCLibrary_Conversation_Blackheart
entry_points [{tag=Conversation.Entry.Blackheart, destinations [#1]}]
nodes [
 ConversationEntryPointNode_7 {
  id=#1
  class=ConversationEntryPointNode
  outputs [#2]
  properties {entryTag {tagName=Conversation.Entry.Blackheart}}
 }
 Speech_C_3 {
  id=#2
  class=Speech_C
  outputs [#3, #4, #5]
  properties {
   nodeName=""
   generalConfig {
    defaultMessage=How did this ship get so far inland?
    contextualMessages [
     {
      message="Not a bad ship, eh? A little wood, a little tar, and she's ready to go."
      contextRequirements [{requirement {class=FortControllerRequirement_IsFirstEverConversationWithNPC}}]
      priority {curve {rowName=Tandem.Speech.Priority.HighTotalBars}}
      weight {value=2}
     }
     {
      message="Yarr, but I do miss the sea."
      requirementMatchPolicy=RequireAny
      contextRequirements [{requirement {
         class=FortControllerRequirement_HasCID
         properties {allowedCIds [
           /BRCosmetics/Athena/Items/Cosmetics/Characters/CID_373_Athena_Commando_M_Pirate01
           /BRCosmetics/Athena/Items/Cosmetics/Characters/CID_372_Athena_Commando_F_Pirate01
           /BRCosmetics/Athena/Items/Cosmetics/Characters/CID_396_Athena_Commando_F_Swashbuckler
          ]}
        }}]
      priority {curve {rowName=Tandem.Speech.Priority.OutfitReaction}}
     }
     {
      message="Yarr, nothin' I love more than watchin' the Academy and lootin' ships."
      contextRequirements [{requirement {class=FortControllerRequirement_HasCID, properties {allowedCIds [&Character_PinkSpike]}}}]
      priority {curve {rowName=Tandem.Speech.Priority.OutfitReaction}}
     }
     {
      message=You look fierce. What flag do ye fly?
      contextRequirements [{requirement {class=FortControllerRequirement_HasCID, properties {allowedCIds [&Character_ChillCat]}}}]
      priority {curve {rowName=Tandem.Speech.Priority.OutfitReaction}}
     }
     {
      message="Best not be boardin' me ship now, bucko."
      contextRequirements [{requirement {
         class=FortControllerRequirement_HasCID
         properties {allowedCIds [&CID_349_Athena_Commando_M_Banana]}
        }}]
      priority {curve {rowName=Tandem.Speech.Priority.OutfitReaction}}
     }
     {
      message=Ye might want to replace those hands with hooks.
      contextRequirements [{requirement {class=FortControllerRequirement_HasCID, properties {allowedCIds [&Character_Bites]}}}]
      priority {curve {rowName=Tandem.Speech.Priority.OutfitReaction}}
     }
     {
      message="What a suit! Ye should resubmit yer application to join me ship. "
      contextRequirements [{requirement {class=FortControllerRequirement_HasCID, properties {allowedCIds [&Character_BadBear]}}}]
      priority {curve {rowName=Tandem.Speech.Priority.OutfitReaction}}
     }
     {
      message="Ye bring an ill wind, don't ye?"
      contextRequirements [{requirement {class=FortControllerRequirement_HasCID, properties {allowedCIds [&Character_TheHerald]}}}]
      priority {curve {rowName=Tandem.Speech.Priority.OutfitReaction}}
     }
     {
      message=Sea legend goes that spiders are lucky aboard ships...
      contextRequirements [{requirement {class=FortControllerRequirement_HasCID, properties {allowedCIds [&Character_RoseDust]}}}]
      priority {curve {rowName=Tandem.Speech.Priority.OutfitReaction}}
     }
    ]
   }
  }
 }
 ConversationLinkNode_2 {
  id=#5
  class=ConversationLinkNode
  properties {remoteEntryTag {tagName=Conversation.Entry.Season22.Narrative.Q08}}
 }
 ConversationLinkNode_0 {id=#4, class=ConversationLinkNode, properties {remoteEntryTag {tagName=Conversation.Entry.BasicServices}}}
 ConversationLinkNode_1 {id=#3, class=ConversationLinkNode, properties {remoteEntryTag {tagName=Conversation.Entry.NPCLibrary.Quests}}}
]
